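# Material function, exposed to the library under Coordinates, that converts a 3D cell index into a 1D index for a grid of given cell counts.
# It floors the index, then sums X, Y times cells X, and Z times cells X squared.
# Cells Y is unused there, so the index assumes equal X and Y counts.
# A second output gives the cell-center UVW by adding half and dividing by the cell counts.
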MaterialFunction /Engine/Functions/Engine_MaterialFunctions01/Coordinates/3Dto1DIndex
graph {
 properties {
  bExposeToLibrary=true
  description="Converts a 3D index into a 1D index based on a given total X,Y,Z size. For example an 3x3x3 grid would return cells 0-26."
  libraryCategoriesText [Coordinates]
 }
 nodes [
  n1 = Floor(None=FunctionInput() {bUsePreviewValueAsDefault=true, inputName=3D Index})
  n2 = MaterialFunctionCall(Float3=@n1) {materialFunction=&BreakOutFloat3Components}
  n3 = FunctionInput() {bUsePreviewValueAsDefault=true, inputName=Cells XYZ, sortPriority=1}
  n4 = MaterialFunctionCall(Float3=@n3) {materialFunction=&BreakOutFloat3Components}
  n5 = Multiply(A=@n2.G, B=@n4.R)
  n6 = Multiply(A=@n4.R, B=@n4.R)
  FunctionOutput(None=Add(A=Add(A=@n2.R, B=@n5), B=Multiply(A=@n2.B, B=@n6))) {outputName=1D Index}
  FunctionOutput(None=Divide(A=Add(A=@n1) {constB=0.5}, B=Multiply(A=@n3))) {outputName=Centered UVW Position, sortPriority=1}
 ]
}
