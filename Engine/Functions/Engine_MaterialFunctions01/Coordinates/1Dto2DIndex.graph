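# Material function that converts a 1D cell index into a 2D grid index for a given Cells XY size, despite its description stating the reverse.
# It rounds the index and X cell count up, then uses modulo for the column and floored division for the row.
# A second output adds a half-cell offset and divides by the grid size to give the cell's centered UV position.
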MaterialFunction /Engine/Functions/Engine_MaterialFunctions01/Coordinates/1Dto2DIndex
graph {
 properties {
  bExposeToLibrary=true
  description="Converts a 2D index into a 1D index based on a given total X,Y size. For example an 8x8 grid would return cells 0-63."
  libraryCategoriesText [Coordinates]
 }
 nodes [
  n1 = Ceil(None=FunctionInput() {bUsePreviewValueAsDefault=true, inputName=1D Index, inputType=FunctionInput_Scalar})
  n2 = FunctionInput() {bUsePreviewValueAsDefault=true, inputName=Cells XY, inputType=FunctionInput_Vector2, sortPriority=1}
  n3 = Ceil(None=ComponentMask(None=@n2) {r=true})
  n4 = MaterialFunctionCall(X=Fmod(A=@n1, B=@n3), Y=Floor(None=Divide(A=@n1, B=@n3))) {materialFunction=&MakeFloat2}
  FunctionOutput(None=@n4.Result) {outputName=2D Index}
  FunctionOutput(None=Divide(A=Add(A=@n4.Result) {constB=0.5}, B=Multiply(A=@n2))) {outputName=Centered UV Position, sortPriority=1}
 ]
}
