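# Material function that converts a 2D grid index into a 1D cell index, exposed in the Coordinates library category.
# It floors the 2D Index input, then adds the X component to Y times the column count, taken from the R channel of Columns X.
# A second output, Centered UV Position, adds 0.5 to the floored index and divides by the Columns X vector to give a normalized cell-center position.
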MaterialFunction /Engine/Functions/Engine_MaterialFunctions01/Coordinates/2Dto1DIndex
graph {
 properties {
  bExposeToLibrary=true
  description="Converts a 2D index into a 1D index based on a given total X,Y size. For example an 8x8 grid would return cells 0-63."
  libraryCategoriesText [Coordinates]
 }
 nodes [
  n1 = Floor(None=FunctionInput() {bUsePreviewValueAsDefault=true, inputName=2D Index, inputType=FunctionInput_Vector2})
  n2 = MaterialFunctionCall(Float2=@n1) {materialFunction=&BreakOutFloat2Components}
  n3 = FunctionInput() {bUsePreviewValueAsDefault=true, inputName=Columns X, inputType=FunctionInput_Vector2, sortPriority=1}
  n4 = ComponentMask(None=@n3) {r=true}
  FunctionOutput(None=Add(A=@n2.R, B=Multiply(A=@n2.G, B=@n4))) {description=The 1D index result, outputName=1D Index}
  FunctionOutput(None=Divide(A=Add(A=@n1) {constB=0.5}, B=Multiply(A=@n3))) {
   description="This is the centered V2 position of the original input, defined as floor(UV) + 0.5."
   outputName=Centered UV Position
   sortPriority=1
  }
 ]
}
