MaterialFunction /Engine/Functions/Engine_MaterialFunctions01/Coordinates/DrawLine-2D
graph {
 properties {bExposeToLibrary=true, description="Draws a line between 2 points using supplied 2D coordinates. "}
 nodes [
  n1 = FunctionInput(Preview=TextureCoordinate() {desc=UVs}) {bUsePreviewValueAsDefault=true, inputName=Coordinates, inputType=FunctionInput_Vector2}
  n2 = FunctionInput() {bUsePreviewValueAsDefault=true, inputName=P0, inputType=FunctionInput_Vector2, sortPriority=1}
  n3 = Subtract(A=@n1, B=@n2)
  n4 = FunctionInput() {bUsePreviewValueAsDefault=true, inputName=P1, inputType=FunctionInput_Vector2, sortPriority=2, previewValue=<1 1 0 1>}
  n5 = Subtract(A=@n4, B=@n2)
  n6 = Normalize(VectorInput=@n5)
  n7 = MaterialFunctionCall(XY=@n6) {materialFunction=&Swizzle}
  n8 = Constant2Vector() {g=-1, r=1}
  n9 = AppendVector(A=DotProduct(A=@n3, B=Multiply(A=@n7.YX, B=@n8)), B=DotProduct(A=@n3, B=@n6)) {desc=Transformed UVs}
  n10 = FunctionInput() {bUsePreviewValueAsDefault=true, inputName=Line Width, inputType=FunctionInput_Scalar, sortPriority=3, previewValue=<0.005 0 0 1>}
  n11 = Subtract(A=Abs(None=@n9), B=@n10)
  n12 = FunctionInput() {bUsePreviewValueAsDefault=true, inputName=Edge Falloff, inputType=FunctionInput_Scalar, sortPriority=4, previewValue=<0.5 1 0 1>}
  n13 = Multiply(A=@n10, B=@n12)
  n14 = OneMinus(None=Clamp(None=Divide(A=@n11, B=@n13)))
  n15 = ComponentMask(None=@n14) {desc=Line Width, r=true}
  n16 = ComponentMask(None=@n9) {desc=Line Length, g=true}
  n17 = Add(A=@n16, B=@n13)
  n18 = Clamp(None=Divide(A=Add(A=@n10, B=@n17), B=@n13))
  n19 = Min(A=Min(A=@n15, B=@n18), B=@n18)
  n20 = Subtract(
   A=Subtract(A=@n16, B=MaterialFunctionCall(Vector 2=@n5) {materialFunction=/Engine/Functions/Engine_MaterialFunctions02/Utility/VectorLength}.V2 Length)
   B=@n10
  )
  n21 = OneMinus(None=Clamp(None=Divide(A=@n20, B=@n13)))
  n22 = Subtract(A=@n1, B=@n2) {desc=PA}
  n23 = Subtract(A=@n4, B=@n2) {desc=BA}
  n24 = Clamp(None=Divide(A=DotProduct(A=@n22, B=@n23), B=DotProduct(A=@n23, B=@n23)) {constB=1})
  n25 = Distance(A=Subtract(A=@n22, B=Multiply(A=@n24, B=@n23)), B=Constant())
  n26 = Clamp(None=Divide(A=Subtract(A=@n25, B=@n10), B=@n13))
  n27 = StaticBool() {value=true}
  FunctionOutput(None=StaticSwitch(
    True=Min(A=@n19, B=@n21)
    False=OneMinus(None=@n26)
    Value=FunctionInput(Preview=@n27) {bUsePreviewValueAsDefault=true, inputName=Square Corners?, inputType=FunctionInput_StaticBool, sortPriority=5}
   ))
  Min(A=@n15, B=ComponentMask(None=@n14) {b=true, desc=Line Width})
  OneMinus()
 ]
}
