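# LongLatToUV is a library material function that converts a direction vector into a UV for long/lat (equirectangular) coordinates, listed under Coordinates and Cubemaps.
# The input Vector is normalized, then passed to a Custom HLSL node that outputs a float2.
# U comes from atan2 of X and negated Y remapped to 0–1, and V from acos of Z divided by pi, making Z the polar axis.
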MaterialFunction /Engine/Functions/Engine_MaterialFunctions01/Coordinates/LongLatToUV
graph {
 properties {
  bExposeToLibrary=true
  description=This node transforms a directional vector into a UV value for use in Long/Lat coordinates
  libraryCategoriesText [Misc,Coordinates,Cubemaps]
 }
 nodes [
  MaterialExpressionNormalize_1 = Normalize(VectorInput=FunctionInput() {description=The vector to transform into a UV value., inputName=Vector, previewValue=<0 0 1 1>})
  n1 = Custom!(InVector=@MaterialExpressionNormalize_1) {
   code="float2((1 + atan2(InVector.x, - InVector.y) / 3.14159265) / 2, acos(InVector.z) / 3.14159265);"
   outputType=CMOT_Float2
   inputs [{inputName=InVector, input {expression=&self:MaterialExpressionNormalize_1}}]
  }
  FunctionOutput(None=@n1)
 ]
}
