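# Material function, exposed to the library, that remaps 0-1 UVs to a specified X and Y range, for controlling range rather than tiling.
# A Vector4 packs X range in R-G and Y range in B-A; two lerps use U and V (default TexCoord) as alphas.
# It outputs combined UV, separate x and y, and the range passed through.
# Size subtractions and Add nodes feed no output; V is not inverted, despite a flip note.
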MaterialFunction /Engine/Functions/Engine_MaterialFunctions01/Coordinates/UVRemap_0-1_ToRange
graph {
 properties {
  bExposeToLibrary=true
  description=Remaps UVs from 0-1 to the specified range. Useful for when you want control of range rather than tiling.
  libraryCategoriesText [Misc,Coordinates]
 }
 nodes [
  n1 = FunctionInput() {
   bUsePreviewValueAsDefault=true
   description="X range=R-G, Y range =B-A"
   inputName=X range Y range
   inputType=FunctionInput_Vector4
   previewValue=<-1 1 -1 1>
  }
  n2 = ComponentMask(None=@n1) {b=true}
  n3 = ComponentMask(None=@n1) {a=true}
  n4 = FunctionInput(Preview=TextureCoordinate()) {
   bUsePreviewValueAsDefault=true
   description=Leave this input alone unless you want to change index or know the exact effect of remapping when the input range is not 0-1.
   inputName=UVs
   inputType=FunctionInput_Vector2
   sortPriority=1
   previewValue=<0 1 -1 1>
  }
  n5 = LinearInterpolate(A=@n2, B=@n3, Alpha=ComponentMask(None=@n4) {g=true}) {constA=-1, desc="Y, Flip so 0 is on bottom"}
  Add(A=@n5)
  n6 = ComponentMask(None=@n1) {r=true}
  n7 = ComponentMask(None=@n1) {g=true}
  n8 = LinearInterpolate(A=@n6, B=@n7, Alpha=ComponentMask(None=@n4) {r=true}) {constA=-1, desc=X}
  Add(A=@n8)
  FunctionOutput(None=@n8) {outputName=x, sortPriority=2}
  FunctionOutput(None=@n5) {outputName=y, sortPriority=4}
  FunctionOutput(None=AppendVector(A=@n8, B=@n5)) {outputName=UV, sortPriority=1}
  FunctionOutput(None=@n1) {outputName=Range}
  Subtract(A=@n7, B=@n6) {desc=X size}
  Subtract(A=@n3, B=@n2) {desc=Y size}
 ]
}
