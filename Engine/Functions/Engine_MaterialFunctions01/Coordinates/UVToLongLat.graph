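# Library material function that converts a UV into a direction vector, so a cubemap can be sampled like a regular 2D longitude-latitude texture.
# The UV input, previewed with texture coordinates, feeds a Custom HLSL node that maps U to a half-turn-offset longitude and V to polar angle.
# The node builds the spherical direction and swizzles it to put the polar axis on Z.
# Separate Add and Pi nodes are unconnected to the output.
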MaterialFunction /Engine/Functions/Engine_MaterialFunctions01/Coordinates/UVToLongLat
graph {
 properties {
  bExposeToLibrary=true
  description=This node transforms a UV value into a directional vector which can be used to sample a Cubemap texture as a regular 2D texture.
  libraryCategoriesText [Misc,Coordinates,Cubemaps]
 }
 nodes [
  n1 = TextureCoordinate()
  Add(A=MaterialFunctionCall(Float2=@n1) {materialFunction=&BreakOutFloat2Components}.R) {constB=0.5}
  MaterialExpressionFunctionInput_0 = FunctionInput(Preview=@n1) {description=The UV to transform into a directional Vector, inputName=UV, inputType=FunctionInput_Vector2, previewValue=<0 0 1 1>}
  n2 = Custom!(UV=@MaterialExpressionFunctionInput_0) {
   code="float2 Angles = float2(2 * PI * (UV.x + 0.5f), PI * UV.y);\r\n\t\r\nfloat s = sin(Angles.y);\r\nfloat3 Direction = float3(s * sin(Angles.x), cos(Angles.y), -s * cos(Angles.x));\r\n\r\nreturn Direction.xzy;"
   inputs [{inputName=UV, input {expression=&self:MaterialExpressionFunctionInput_0}}]
  }
  FunctionOutput(None=@n2)
  MaterialFunctionCall(Multiplier=Constant() {r=2}) {materialFunction=&Pi}
 ]
}
