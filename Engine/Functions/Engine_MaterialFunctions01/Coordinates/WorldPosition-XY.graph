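# WorldPosition-XY is a library-exposed material function that outputs world position XY coordinates divided by a Scale input.
# It divides the world position by the scalar Scale, which falls back to its preview value when unconnected, then masks the result to the red and green channels.
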MaterialFunction /Engine/Functions/Engine_MaterialFunctions01/Coordinates/WorldPosition-XY
graph {
 properties {bExposeToLibrary=true, description=WorldPosition XY Coordinates with Scale input.}
 nodes [
  n1 = WorldPosition()
  n2 = FunctionInput() {bUsePreviewValueAsDefault=true, inputName=Scale, inputType=FunctionInput_Scalar, previewValue=<100 0 0 1>}
  FunctionOutput(None=ComponentMask(None=Divide(A=@n1.XYZ, B=@n2)) {g=true, r=true})
 ]
}
