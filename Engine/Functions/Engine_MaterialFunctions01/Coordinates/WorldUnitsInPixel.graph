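# WorldUnitsInPixel is a library material function that converts a world-space size into its on-screen size in pixels at a given depth.
# A Custom node reads the X cotangent of half the field of view, which is multiplied by view size, masked to the horizontal component and halved.
# That factor is divided by the Depth input, which defaults to PixelDepth, then scaled by WorldUnits to output Pixels.
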MaterialFunction /Engine/Functions/Engine_MaterialFunctions01/Coordinates/WorldUnitsInPixel
graph {
 properties {bExposeToLibrary=true}
 nodes [
  n1 = FunctionInput() {bUsePreviewValueAsDefault=true, inputName=WorldUnits, inputType=FunctionInput_Scalar, previewValue=<1 0 0 1>}
  n2 = Custom!(None) {code="GetCotanHalfFieldOfView().x", desc="GetCotanHalfFieldOfView()", outputType=CMOT_Float2}
  n3 = ViewSize()
  n4 = Multiply(A=ComponentMask(None=Multiply(A=@n2, B=@n3)) {r=true}) {constB=0.5}
  n5 = FunctionInput(Preview=PixelDepth().R) {bUsePreviewValueAsDefault=true, desc=">=0", description=from camera e.g. PixelDepth, inputName=Depth, inputType=FunctionInput_Scalar}
  FunctionOutput(None=Multiply(A=@n1, B=Divide(A=@n4, B=@n5) {desc=Orig Particle Radius in pixels})) {outputName=PIxels}
 ]
}
